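; BP_TaleSecret randomly selects one of three "tale secrets" for a Fortnite map, apparently in a test setup.
; BeginPlay stores the random pick and sets ActivateSecret, with no authority check in the graph.
; OnRep_ActivateSecret makes the chosen secret's root component visible and silently kills the two chests belonging to the other secrets.
; Debug prints report the pick (boat, arcade or rope branch) or a failed match.
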(event EventBeginPlay
  (bind _element (ArrayMacros|GetRandomArrayElement (Utilities|Array|MakeArray (Variables|Default|GetTaleSecret_0) (Variables|Default|GetTaleSecret_1) (Variables|Default|GetTaleSecret_2)))
    (:%20%20
      (bind _output_get (Variables|Default|SetTaleSecret_Chosen _element))
      (Development|PrintString (Utilities|GetDisplayName _output_get) true true "(R=1.000000,G=0.078184,B=0.000000,A=1.000000)")
      (Variables|Default|SetActivateSecret true))))

(fn OnRep_ActivateSecret
  (Rendering|SetVisibility (Variables|/Script/Engine.Actor|GetRootComponent (Variables|Default|GetTaleSecret_Chosen)) true false)
  (if (== (Variables|Default|GetTaleSecret_Chosen) (Variables|Default|GetTaleSecret_0))
    (Gameplay|SilentDie (Variables|Default|GetTaleSecretChest_1) false :!class /Script/FortniteGame.BuildingActor)
    (Gameplay|SilentDie (Variables|Default|GetTaleSecretChest_2) false :!class /Script/FortniteGame.BuildingActor)
    (Development|PrintString "BOOOOOOOOOOOAAAAAAAAAATTTTTTTT" true true "(R=1.000000,G=0.027444,B=0.000000,A=1.000000)")
    (elif (== (Variables|Default|GetTaleSecret_Chosen) (Variables|Default|GetTaleSecret_1))
      (Gameplay|SilentDie (Variables|Default|GetTaleSecretChest_0) false :!class /Script/FortniteGame.BuildingActor)
      (Gameplay|SilentDie (Variables|Default|GetTaleSecretChest_2) false :!class /Script/FortniteGame.BuildingActor)
      (Development|PrintString "ARCAAAAAAAAAAAAAADDDDEEEEEEEEEEEEEEEEEE" true true "(R=1.000000,G=0.008613,B=0.000000,A=1.000000)")
      (elif (== (Variables|Default|GetTaleSecret_Chosen) (Variables|Default|GetTaleSecret_2))
        (Gameplay|SilentDie (Variables|Default|GetTaleSecretChest_0) false :!class /Script/FortniteGame.BuildingActor)
        (Gameplay|SilentDie (Variables|Default|GetTaleSecretChest_1) false :!class /Script/FortniteGame.BuildingActor)
        (Development|PrintString "ROOOOOOOOOPPPPPPPPPEEEEEEEEEEEEEE" true true "(R=1.000000,G=0.067445,B=0.000000,A=1.000000)")
        (else
          (Development|PrintString "ALLLLLLLLL FAAAAAAAIILLLEEEEEEDDDDD" true true "(R=1.000000,G=0.051730,B=0.000000,A=1.000000)"))))))

(fn ConstructionScript)
